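; Debug actor that adds or removes the Specter mode gameplay effect on all player pawns for movement testing.
; The custom event ApplySpecterEffectToAllPawns runs only with authority and passes each Fort player pawn and its Apply flag to ApplySpecterEffectToPawn.
; That function prints which pawn is affected, then applies GE_InSpecterMode or removes all its stacks through the pawn's ability system component.
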(event Custom|ApplySpecterEffectToAllPawns
  (params (Apply :type Boolean))
  ;[0 0 0 0.5] Add/Remove Specter Effect
  (Networking|SwitchHasAuthority
    (:Authority
      (bind _outfortplayerpawns (Fortnite|GetFortPlayerPawns))
      (for _array_element _outfortplayerpawns
        (CallFunction|ApplySpecterEffecttoPawn _array_element Apply)))))

(fn ApplySpecterEffectToPawn
  (params (Actor :type Actor%20Object%20Reference) (Add :type Boolean))
  (Development|PrintString (Utilities|String|Append (Utilities|String|SelectString "Applying Specter effect to " "Removing Specter effect from " Add) (Fortnite|GetPlayerorObjectNameSafe Actor)))
  (if (Variables|GetAdd)
    (Fortnite|ApplyGameplayEffectsToActor (Variables|GetActor) (Utilities|Array|MakeArray (Variables|Default|GetGE_InSpecterMode)) (Transformation|GetActorLocation (Variables|GetActor)) false 0 (Ability|GetFortAbilitySystemComponent (Variables|GetActor)))
    (else
      (GameplayEffects|RemoveActiveGameplayEffectbySourceEffect (Ability|GetFortAbilitySystemComponent (Variables|GetActor)) (Variables|Default|GetGE_InSpecterMode) 0 -1))))

(fn ConstructionScript)
